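; Front-end preview display for skydive contrail cosmetics, posing the displayed hero pawn in a skydive so the trail can be shown.
; On preview it binds the contrail activation event, plays a gender-specific dive montage and refreshes contrail visibility.
; On activation and every tick it applies Niagara or Cascade parameters, chosen by component cast.
; Tick maps the ParaGlideLeanAlpha curve to fake dive velocity; the Cascade path carries many per-contrail overrides.
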(event SkyDiveContrails|EventPreviewContrail
  (params (Contrail :type Athena%20Sky%20Dive%20Contrail%20Item%20Definition%20Object%20Reference))
  (bind _playerpawn (Variables|Default|GetPlayerPawn))
  (bind _mesh (Variables|/Script/Engine.Character|GetMesh _playerpawn))
  (bind _returnvalue (FrontEndContext|GetHeroPlayerPawnforCurrentDisplayedItem (BlueprintContext|GetFortFrontEndContext)))
  (bind _asplayer_pawn_athena (Utilities|Casting|CastToPlayerPawn_Athena _returnvalue))
  (Variables|Default|SetPlayerPawn _asplayer_pawn_athena)
  (CallFunction|BindContrailsEvents)
  ;[1 1 1] Get Contrail Soft Path - will return a Cascase or Niagra Particle system soft object ptr based on the configuration of the contrail definition and the hotfix variable 'fx.FavorNiagaraOverCascade'
  (Animation|ResetClothandDynamics (Variables|Default|GetPlayerPawn))
  (bind _returnvalue_2 (Animation|Utilities|GetFrontEndAnimInstance _mesh))
  (Utilities|IsValid _returnvalue_2
    (:Is%20Valid
      (Variables|/Script/FortniteGame.FrontendAnimInstance|SetbIsSkydiving true _returnvalue_2)
      (bind _output_get (Variables|Default|SetDiveAnimation (select (FortPlayerPawn|GetGender _playerpawn) "/Game/Athena/Cosmetics/Contrails/SkydivePreview_mix_M.SkydivePreview_mix_M" "/Game/Athena/Cosmetics/Contrails/SkydivePreview_mix_F.SkydivePreview_mix_F"))))
    (:Is%20Not%20Valid
      _))
  (Animation|Montage|MontagePlay (Components|SkeletalMesh|GetAnimInstance _mesh) _output_get 1.0 "MontageLength" 0.0 true)
  (Utilities|FlowControl|Delay 0.2)
  (Class|FortContrailsComponent|RefreshContrailVisibility (Variables|/Game/Athena/PlayerPawn_Athena.PlayerPawn_Athena_C|GetFortContrails (Variables|Default|GetPlayerPawn))))

(event Custom|OnContrailsActivated
  (CallFunction|UnbindContrailsEvents)
  (bind _asniagara_particle_system_component (Utilities|Casting|CastToNiagaraComponent (Class|FortContrailsComponent|GetContrailFXComponent (Variables|/Game/Athena/PlayerPawn_Athena.PlayerPawn_Athena_C|GetFortContrails (Variables|Default|GetPlayerPawn))))
    (:then
      (CallFunction|ApplyNiagaraParams _asniagara_particle_system_component))
    (:CastFailed
      (CallFunction|ApplyCascadeParams 0))))

(event EventTick
  (params (DeltaSeconds :type Float%20%28single-precision%29))
  (bind _playerpawn (Variables|Default|GetPlayerPawn))
  (bind _mesh (Variables|/Script/Engine.Character|GetMesh _playerpawn))
  (bind _returnvalue (Components|SkeletalMesh|GetAnimInstance _mesh))
  (bind _returnvalue_2 (Animation|Curves|GetCurveValue _returnvalue "ParaGlideLeanAlpha"))
  (bind _returnvalue_3 (FrontEndContext|GetHeroPlayerPawnforCurrentDisplayedItem (BlueprintContext|GetFortFrontEndContext)))
  (bind _asplayer_pawn_athena (Utilities|Casting|CastToPlayerPawn_Athena _returnvalue_3))
  (Variables|Default|SetPlayerPawn _asplayer_pawn_athena)
  (Variables|Default|SetParaGlideLeanAlpha _returnvalue_2)
  (Variables|Default|SetDiveSpeed_VelocityY (Math|Float|MapRangeClamped _returnvalue_2 0.0 1.0 (Variables|Default|GetTargetGlide_VelocityY) (Variables|Default|GetTargetDive_VelocityY)))
  (bind _asniagara_particle_system_component (Utilities|Casting|CastToNiagaraComponent (Class|FortContrailsComponent|GetContrailFXComponent (Variables|/Game/Athena/PlayerPawn_Athena.PlayerPawn_Athena_C|GetFortContrails (Variables|Default|GetPlayerPawn))))
    (:then
      (CallFunction|ApplyNiagaraParams _asniagara_particle_system_component))
    (:CastFailed
      (CallFunction|ApplyCascadeParams 0))))

(fn ApplyCascadeParams
  (params (FXComp :type FXSystem%20Component%20Object%20Reference))
  (bind _velocityz (Variables|Default|GetVelocityZ))
  (bind _returnvalue (Math|Float|Absolute(Float) _velocityz))
  (bind _divespeed_velocityy (Variables|Default|GetDiveSpeed_VelocityY))
  (Effects|Components|ParticleSystem|SetFloatParameter FXComp "TrailAlpha" (Math|Float|MapRangeClamped _returnvalue 1200 3300 0.1 0.7))
  (Effects|Components|ParticleSystem|SetVectorParameter FXComp "TrailWidth" (Math|Vector|MakeVector (Math|Float|MapRangeClamped _returnvalue 1200 3300 0.75 1.25)))
  (Effects|Components|ParticleSystem|SetVectorParameter FXComp "TestVelocity" (Math|Vector|MakeVector 0.0 (Variables|Default|GetDiveSpeed_VelocityY) (Variables|Default|GetVelocityZ)))
  (Effects|Components|ParticleSystem|SetFloatParameter FXComp "RibbonSpawnRate" 1.2e+02)
  (Effects|Components|ParticleSystem|SetFloatParameter FXComp "Spawn_FrontEnd" 1e+01)
  (Effects|Components|ParticleSystem|SetFloatParameter FXComp "DontSpawn_FrontEnd" 0.0)
  (Effects|Components|ParticleSystem|SetFloatParameter FXComp "DontSpawn_FrontEnd" 1.0)
  ;[1 0.3666 0.9334] Heart Contrail
  (Effects|Components|ParticleSystem|SetVectorParameter FXComp "HeartTestVelocity" (Math|Vector|MakeVector 0.0 (Variables|Default|GetDiveSpeed_VelocityY) 2e+03))
  (Effects|Components|ParticleSystem|SetVectorParameter FXComp "HeartSmokeSize" "0.250000,0.250000,0.250000")
  (Effects|Components|ParticleSystem|SetVectorParameter FXComp "HeartSmokeLocation" "0.000000,-100.000000,0.000000")
  ;[0.1411 1 0] Retro SciFi Contrail
  (Effects|Components|ParticleSystem|SetFloatParameter FXComp "RetroRayGunSpawn" 5e+01)
  ;[0 0.3188 1] Lightning Contrail
  (Effects|Components|ParticleSystem|SetFloatParameter FXComp "LightningSpawnRate" 5e+01)
  (Effects|Components|ParticleSystem|SetVectorParameter FXComp "StrokeOffset" "0, 0,-20.000000")
  ;[0.315 0.4792 1] Comet Contrail
  (Effects|Components|ParticleSystem|SetFloatParameter FXComp "SparkSpawn" 25.0)
  (Effects|Components|ParticleSystem|SetVectorParameter FXComp "FrontEndSize" (Math|Vector|MakeVector (Math|Float|MapRangeClamped (Variables|Default|GetDiveSpeed_VelocityY) (Variables|Default|GetTargetGlide_VelocityY) (Variables|Default|GetTargetDive_VelocityY) 0.8 0.52) (Math|Float|MapRangeClamped (Variables|Default|GetDiveSpeed_VelocityY) (Variables|Default|GetTargetGlide_VelocityY) (Variables|Default|GetTargetDive_VelocityY) 0.8 0.52) (Math|Float|MapRangeClamped (Variables|Default|GetDiveSpeed_VelocityY) (Variables|Default|GetTargetGlide_VelocityY) (Variables|Default|GetTargetDive_VelocityY) 0.8 0.52))) ; Original vector all 0.75
  ;[1 0.3735 0.8645] Spray Paint Contrail
  (Effects|Components|ParticleSystem|SetVectorParameter FXComp "SparyColorIntensity" "2.000000,2.000000,2.000000")
  ;[1 1 1] Toilet Paper Contrail - Cleaned up two unuseful Set Parameter Nodes
  (Effects|Components|ParticleSystem|SetVectorParameter FXComp "ToiletPaperFootOffset" "0.000000,0.000000,-5.000000")
  ;[0.004252 0 1] Rune Contrail
  (Effects|Components|ParticleSystem|SetFloatParameter FXComp "VikingRuneMotion" 0.5)
  (Effects|Components|ParticleSystem|SetVectorParameter FXComp "AmmoBeltTestVelocity" (Math|Vector|MakeVector 2.0 (Variables|Default|GetDiveSpeed_VelocityY) 2.0)) ; This is a guess
  ;[0 0.877 1] Ice Contrail
  (Effects|Components|ParticleSystem|SetVectorParameter FXComp "IceTrailOffset" "0, 0,-15.000000")
  ;[1 1 1] Rainbow Contrail
  (Effects|Components|ParticleSystem|SetVectorParameter FXComp "RainbowHandOffset" "0,0.000000,-20.000000")
  (Effects|Components|ParticleSystem|SetVectorParameter FXComp "RainbowFootOffset" "0, 0,-20.000000")
  ;[1 1 1] Blue Streak Offset
  (Effects|Components|ParticleSystem|SetVectorParameter FXComp "BlueHandOffset" "0.000000, 0,-30.000000")
  (Effects|Components|ParticleSystem|SetVectorParameter FXComp "BlueFootOffset" "0, 0,-30.000000")
  (Effects|Components|ParticleSystem|SetVectorParameter FXComp "StarsHandOffset" "0.000000, 0,-30.000000")
  (Effects|Components|ParticleSystem|SetVectorParameter FXComp "StarsFootOffset" "0, 0,-30.000000")
  ;[0 0 0] Unit parameter. We could give detailed values in cascade. -Yue.Jia
  (Effects|Components|ParticleSystem|SetVectorParameter FXComp "FrontendPreviewVector" "1.000000,1.000000,1.000000")
  (Effects|Components|ParticleSystem|SetFloatParameter FXComp "FrontendPreviewFloat" 1.0)
  (Rendering|Material|SetScalarParameterValue "/Game/Packages/Fortress_SharedMaterials/GlobalMaterialParameters/FortniteMaterialParameters.FortniteMaterialParameters" "AthenaFrontendPreview" 1.0 :!class /Script/Engine.KismetMaterialLibrary)
  ;[0.02824 0 0.07292] Raven Skydive Trail_Zhiyang.Zhang
  (Effects|Components|ParticleSystem|SetVectorParameter FXComp "TestVelocityFeathers" (Math|Vector|MakeVector 0.0 (Math|Float|MapRangeClamped (Variables|Default|GetDiveSpeed_VelocityY) (Variables|Default|GetTargetGlide_VelocityY) (Variables|Default|GetTargetDive_VelocityY) 0.0 -1e+02) (* (Variables|Default|GetVelocityZ) 0.1)))
  (Effects|Components|ParticleSystem|SetFloatParameter FXComp "VBuckSpawn" 3e+01)
  (Effects|Components|ParticleSystem|SetVectorParameter FXComp "GreenSmokeHandOffset" "0.000000,0.000000,-6.000000")
  (Effects|Components|ParticleSystem|SetVectorParameter FXComp "GreenSmokeFootOffset" "0.000000, 0,-6.000000")
  ;[0.3564 1 0.2417] Green Flame Skydive Trail_Zhiyang.Zhang
  (Effects|Components|ParticleSystem|SetVectorParameter FXComp "TestVelocityFlame" (Math|Vector|MakeVector 0.0 0.0 (* (Variables|Default|GetVelocityZ) 0.07)))
  ;[0 0.3435 1] Snowflake Skydive Trail_Zhiyang.Zhang
  (Effects|Components|ParticleSystem|SetVectorParameter FXComp "TestVelocitysnowlights" (Math|Vector|MakeVector 0.0 (Math|Float|MapRangeClamped (Variables|Default|GetDiveSpeed_VelocityY) (Variables|Default|GetTargetGlide_VelocityY) (Variables|Default|GetTargetDive_VelocityY) 0.0 -125.0) (* (Variables|Default|GetVelocityZ) 0.12)))
  ;[1 1 1] StringLights Skydive Trail_Zhiyang.Zhang
  (Effects|Components|ParticleSystem|SetFloatParameter FXComp "TestVelocitystringlights" -4.5e+02)
  ;[1 1 1] FiberOptics Skydive Trail_Zhiyang.Zhang
  (Effects|Components|ParticleSystem|SetFloatParameter FXComp "TestVelocityfiberoptics" -4.5e+02) ; original value -450
  ;[0 0.5877 1] Raven Skydive Trail_Zhiyang.Zhang
  (Effects|Components|ParticleSystem|SetVectorParameter FXComp "TestVelocityConfetti" (Math|Vector|MakeVector 0.0 0.0 (* (Variables|Default|GetVelocityZ) 0.3)))
  ;[1 0.1381 0] Lava Skydive Trail_Zhiyang.Zhang
  (Effects|Components|ParticleSystem|SetVectorParameter FXComp "TestVelocitylava" (Math|Vector|MakeVector 0.0 (Math|Float|MapRangeClamped (Variables|Default|GetDiveSpeed_VelocityY) (Variables|Default|GetTargetGlide_VelocityY) (Variables|Default|GetTargetDive_VelocityY) 0.0 -2e+02) (* (Variables|Default|GetVelocityZ) 0.18)))
  (Effects|Components|ParticleSystem|SetVectorParameter FXComp "TestLocationLava" "0, 0,-6.000000")
  ;[0 0.064 1] battlesuit
  (Effects|Components|ParticleSystem|SetVectorParameter FXComp "BSuitHandOffset" "0,5.000000,-20.000000")
  (Effects|Components|ParticleSystem|SetVectorParameter FXComp "BSuitFootOffset" "0,2.000000,-33.000000")
  (Effects|Components|ParticleSystem|SetFloatParameter FXComp "BSuitSpawnRate" 8e+01)
  (Effects|Components|ParticleSystem|SetFloatParameter FXComp "BSuitSpawnRateFeet" 8e+01)
  (Effects|Components|ParticleSystem|SetVectorParameter FXComp "BSuitFootVelocity" (Math|Vector|MakeVector 0.0 (Math|Float|Clamp(Float) _divespeed_velocityy -3.5e+02 -1.5e+02) 8e+02))
  (Effects|Components|ParticleSystem|SetVectorParameter FXComp "BSuitHandVelocity" (Math|Vector|MakeVector 0.0 (Math|Float|Clamp(Float) _divespeed_velocityy -2e+02 -1.5e+02) 8e+02))
  ;[1 1 1] Swirly Colorful Trail
  (Effects|Components|ParticleSystem|SetVectorParameter FXComp "SwirlyColorfulHandOffset" (Variables|Default|GetSwirlyColorfulHandOffsetParam))
  (Effects|Components|ParticleSystem|SetVectorParameter FXComp "SwirlyColorfulFootOffset" (Variables|Default|GetSwirlyColorfulFootOffsetParam))
  ;[0.1875 1 0.979] SlurpMonster
  (Effects|Components|ParticleSystem|SetVectorParameter FXComp "SlurpMonHandOffset" "0,10.000000,-30.000000")
  (Effects|Components|ParticleSystem|SetFloatParameter FXComp "SlurpRibbonSpawnRate" 5e+01))

(fn ApplyNiagaraParams
  (params (NiagaraParticleSystem :type Niagara%20Particle%20System%20Component%20Object%20Reference))
  (Niagara|SetNiagaraVariable(Bool) NiagaraParticleSystem "User.bIsFrontEndPreview" true)
  (Niagara|SetNiagaraVariable(Vector3) NiagaraParticleSystem "User.FEVec" (Math|Vector|MakeVector 0.0 (Math|Float|MapRangeClamped (Variables|Default|GetDiveSpeed_VelocityY) (Variables|Default|GetTargetGlide_VelocityY) (Variables|Default|GetTargetDive_VelocityY) 0.0 -2e+02) (* (Variables|Default|GetVelocityZ) 0.18)))
  (Niagara|SetNiagaraVariable(Float) NiagaraParticleSystem "User.ParaGlideLeanAlpha" (Variables|Default|GetParaGlideLeanAlpha)))

(fn BindContrailsEvents
  (bind _playerpawn (Variables|Default|GetPlayerPawn))
  (Utilities|IsValid _playerpawn
    (:Is%20Valid
      (EventDispatchers|BindEventtoOnContrailsActivated (Variables|/Game/Athena/PlayerPawn_Athena.PlayerPawn_Athena_C|GetFortContrails _playerpawn) (EventDispatchers|CreateEvent :!event OnContrailsActivated)))))

(fn UnbindContrailsEvents
  (bind _playerpawn (Variables|Default|GetPlayerPawn))
  (Utilities|IsValid _playerpawn
    (:Is%20Valid
      (EventDispatchers|UnbindallEventsfromOnContrailsActivated (Variables|/Game/Athena/PlayerPawn_Athena.PlayerPawn_Athena_C|GetFortContrails _playerpawn)))))

(fn ConstructionScript)
