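; Anim notify state that flags a mesh's front-end anim instance as skydiving for the notify's duration, apparently to preview skydive cosmetics such as contrails.
; Begin and Tick look up the FrontendAnimInstance from the mesh component and, if valid, set bIsSkydiving true, so the flag is reasserted every tick.
; End does the same lookup and clears the flag.
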(fn Received_NotifyBegin
  (params (MeshComp :type Skeletal%20Mesh%20Component%20Object%20Reference) (Animation :type Anim%20Sequence%20Base%20Object%20Reference) (TotalDuration :type Float%20%28single-precision%29) (EventReference :type Anim%20Notify%20Event%20Reference%20Structure%20%28by%20ref%29))
  (bind _returnvalue (Animation|Utilities|GetFrontEndAnimInstance MeshComp))
  (Utilities|IsValid _returnvalue
    (:Is%20Valid
      (Variables|/Script/FortniteGame.FrontendAnimInstance|SetbIsSkydiving true _returnvalue)
      (return false))))

(fn Received_NotifyEnd
  (params (MeshComp :type Skeletal%20Mesh%20Component%20Object%20Reference) (Animation :type Anim%20Sequence%20Base%20Object%20Reference) (EventReference :type Anim%20Notify%20Event%20Reference%20Structure%20%28by%20ref%29))
  (bind _returnvalue (Animation|Utilities|GetFrontEndAnimInstance MeshComp))
  (Utilities|IsValid _returnvalue
    (:Is%20Valid
      (Variables|/Script/FortniteGame.FrontendAnimInstance|SetbIsSkydiving false _returnvalue)
      (return false))))

(fn Received_NotifyTick
  (params (MeshComp :type Skeletal%20Mesh%20Component%20Object%20Reference) (Animation :type Anim%20Sequence%20Base%20Object%20Reference) (FrameDeltaTime :type Float%20%28single-precision%29) (EventReference :type Anim%20Notify%20Event%20Reference%20Structure%20%28by%20ref%29))
  (bind _returnvalue (Animation|Utilities|GetFrontEndAnimInstance MeshComp))
  (Utilities|IsValid _returnvalue
    (:Is%20Valid
      (Variables|/Script/FortniteGame.FrontendAnimInstance|SetbIsSkydiving true _returnvalue)
      (return false))))
